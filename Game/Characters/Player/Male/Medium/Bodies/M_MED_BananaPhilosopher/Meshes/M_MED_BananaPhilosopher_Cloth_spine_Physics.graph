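# Physics asset for the spine cloth bone chain on the medium male BananaPhilosopher player body.
# Four small capsules on dyn_Cloth_spine_00 to 03 form a chain hung from the pelvis by angle-limited constraints, with tighter twist at the root.
# Larger unconstrained capsules cover the pelvis, spine, left thigh and calf, plus dyn_Cloth_volume helper bones, apparently as collision volumes for the cloth.
# Leg coverage is left side only.
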PhysicsAsset /Game/Characters/Player/Male/Medium/Bodies/M_MED_BananaPhilosopher/Meshes/M_MED_BananaPhilosopher_Cloth_spine_Physics
bodies {
 calf_l [{shapeType=Capsule, length=31.5, radius=13.5, shapeName=calf_l_capsule, center=<-28.347 0.0001 0>, rotation=<rot 90 0.0009 0.001>}]
 dyn_Cloth_spine_00 [{
   shapeType=Capsule
   length=3.5643
   radius=4.2097
   shapeName=dyn_Cloth_spine_00_capsule
   center=<-1.4893 -1.5491 0.1303>
   rotation=<rot 85.0001 0.0001 0>
  }]
 dyn_Cloth_spine_01 [{shapeType=Capsule, length=6.5, radius=4.5, shapeName=dyn_Cloth_spine_01_capsule, center=<-5.1363 -2.5152 0>, rotation=<rot 90 90 89.9999>}]
 dyn_Cloth_spine_02 [{shapeType=Capsule, length=6, radius=5, shapeName=dyn_Cloth_spine_02_capsule, center=<-5.1363 -1.8015 0>, rotation=<rot 90 -180 179.9999>}]
 dyn_Cloth_spine_03 [{shapeType=Capsule, length=5, radius=6, shapeName=dyn_Cloth_spine_03_capsule, center=<-6.2109 -0.4311 0>, rotation=<rot 90 179.6835 179.6834>}]
 dyn_Cloth_volume_fn [{shapeType=Capsule, length=6.8692, radius=10.9617, shapeName=dyn_Cloth_volume_fn_capsule, center=<2.2275 7.2341 0>}]
 dyn_Cloth_volume_fn01 [{shapeType=Capsule, length=8.3224, radius=10.4953, shapeName=dyn_Cloth_volume_fn01_capsule, center=<-8.4823 8.9932 0>}]
 dyn_Cloth_volume_fn_00 [{shapeType=Capsule, length=6.8692, radius=10.9617, shapeName=dyn_Cloth_volume_fn_capsule, center=<-11.0636 8.5147 0>}]
 dyn_Cloth_volume_l_A01 [{
   shapeType=Capsule
   length=0
   radius=11
   shapeName=dyn_Cloth_volume_l_A01_capsule
   center=<-17.2011 14.7177 -5.2887>
   rotation=<rot 0 0.0001 -10>
  }]
 pelvis [{shapeType=Capsule, length=7.3108, radius=22.2812, shapeName=pelvis_capsule, center=<-1.8316 -3.6597 0>}]
 spine_01 [{shapeType=Capsule, length=6.9606, radius=17.1349, shapeName=spine_01_capsule, center=<5.4994 -3.6523 0>}]
 spine_02 [{shapeType=Capsule, length=5.0161, radius=20.8279, shapeName=spine_02_capsule, center=<0 -3.0923 0>}]
 spine_04 [{shapeType=Capsule, length=3.5, radius=20.6045, shapeName=spine_04_capsule, center=<0 -5.4882 0>}]
 thigh_l [{
   shapeType=Capsule
   length=28.5847
   radius=16.3129
   shapeName=thigh_l_capsule
   center=<-16.0051 -5.3438 2.5037>
   rotation=<rot -88.0732 -0.0024 0.0019>
  }]
}
constraints [
 {bone1Name=dyn_Cloth_spine_00, bone2Name=pelvis, swing1LimitDegrees=45, swing2LimitDegrees=45, twistLimitDegrees=20}
 {bone1Name=dyn_Cloth_spine_01, bone2Name=dyn_Cloth_spine_00, swing1LimitDegrees=45, swing2LimitDegrees=45, twistLimitDegrees=45}
 {bone1Name=dyn_Cloth_spine_02, bone2Name=dyn_Cloth_spine_01, swing1LimitDegrees=45, swing2LimitDegrees=45, twistLimitDegrees=45}
 {bone1Name=dyn_Cloth_spine_03, bone2Name=dyn_Cloth_spine_02, swing1LimitDegrees=45, swing2LimitDegrees=45, twistLimitDegrees=45}
]
