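# Physics asset for the Sleek mesh of the male medium BananaPhilosopher player body, setting up physics for its lower-garment cloth-volume bones.
# Front, back and six side chains of three capsules hang from the pelvis, with an extra left-side spine capsule.
# Cross-constraints tie neighboring side chains mid-length and join rear-side tips to the back tip with free rotation.
# Unconstrained capsules on pelvis, spine, thighs and calves provide body collision shapes.
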PhysicsAsset /Game/Characters/Player/Male/Medium/Bodies/M_MED_BananaPhilosopher/Meshes/M_MED_BananaPhilosopher_Sleek_Physics
bodies {
 calf_l [{
   shapeType=Capsule
   length=42.4623
   radius=12.4608
   shapeName=calf_l_capsule
   center=<-27.9318 -1.6458 -0.5062>
   rotation=<rot -87.8858 -0.0001 0.0001>
  }]
 calf_r [{
   shapeType=Capsule
   length=40.7551
   radius=12.7101
   shapeName=calf_r_capsule
   center=<27.1038 1.1039 1.0006>
   rotation=<rot -87.8858 -0.0002 0.0001>
  }]
 dyn_Cloth_volume_bk [{shapeType=Capsule, length=12.5884, radius=7.7451, shapeName=dyn_Cloth_volume_bk_capsule, center=<-15.0543 6.68 0>}]
 dyn_Cloth_volume_bk01 [{shapeType=Capsule, length=10.9597, radius=11.3355, shapeName=dyn_Cloth_volume_bk01_capsule, center=<-20.6687 7.1556 0>}]
 dyn_Cloth_volume_bk_00 [{shapeType=Capsule, length=12.9232, radius=9.8852, shapeName=dyn_Cloth_volume_bk_capsule, center=<-19.0392 5.3658 0>}]
 dyn_Cloth_volume_fn [{shapeType=Capsule, length=18.5858, radius=7.7451, shapeName=dyn_Cloth_volume_fn_capsule, center=<-8.309 9.0401 0>}]
 dyn_Cloth_volume_fn01 [{shapeType=Capsule, length=16.539, radius=9.2788, shapeName=dyn_Cloth_volume_fn01_capsule, center=<-19.0188 6.1685 0>}]
 dyn_Cloth_volume_fn_00 [{shapeType=Capsule, length=18.5858, radius=7.7451, shapeName=dyn_Cloth_volume_fn_capsule, center=<-21.6001 7.7445 0>}]
 dyn_Cloth_volume_l_A [{
   shapeType=Capsule
   length=20.2232
   radius=11.9541
   shapeName=dyn_Cloth_volume_l_A_capsule
   center=<-8.309 8.9922 3.5395>
   rotation=<rot 0 0.0002 20>
  }]
 dyn_Cloth_volume_l_A01 [{
   shapeType=Capsule
   length=5.9466
   radius=10.6296
   shapeName=dyn_Cloth_volume_l_A01_capsule
   center=<-19.0188 6.1206 3.5395>
   rotation=<rot 0 0.0001 25>
  }]
 dyn_Cloth_volume_l_A_00 [{
   shapeType=Capsule
   length=7.9934
   radius=9.096
   shapeName=dyn_Cloth_volume_l_A_capsule
   center=<-21.6001 6.1206 3.5395>
   rotation=<rot 0 0.0001 25>
  }]
 dyn_Cloth_volume_l_B [{shapeType=Capsule, length=12.5884, radius=7.7451, shapeName=dyn_Cloth_volume_l_B_capsule, center=<-6.5604 -1.772 0>}]
 dyn_Cloth_volume_l_B01 [{shapeType=Capsule, length=10.9597, radius=11.3355, shapeName=dyn_Cloth_volume_l_B01_capsule, center=<-19.6493 -6.1459 0>}]
 dyn_Cloth_volume_l_B_00 [{shapeType=Capsule, length=12.5884, radius=7.7451, shapeName=dyn_Cloth_volume_l_B_capsule, center=<-15.5094 -1.772 0>}]
 dyn_Cloth_volume_l_C [{shapeType=Capsule, length=12.5884, radius=7.7451, shapeName=dyn_Cloth_volume_l_C_capsule, center=<-15.0543 6.68 0>}]
 dyn_Cloth_volume_l_C01 [{shapeType=Capsule, length=10.9597, radius=11.3355, shapeName=dyn_Cloth_volume_l_C01_capsule, center=<-20.6687 7.1556 0>}]
 dyn_Cloth_volume_l_C_00 [{shapeType=Capsule, length=12.9232, radius=9.8852, shapeName=dyn_Cloth_volume_l_C_capsule, center=<-19.0392 5.3658 0>}]
 dyn_Cloth_volume_l_C_spine [{
   shapeType=Capsule
   length=10.5481
   radius=7.6773
   shapeName=dyn_Cloth_volume_l_C01_capsule
   center=<-10.4956 -4.0863 0>
   rotation=<rot -90 180 179.9999>
  }]
 dyn_Cloth_volume_r_A [{shapeType=Capsule, length=12.5884, radius=7.7451, shapeName=dyn_Cloth_volume_r_A_capsule, center=<-6.5604 4.8634 0>}]
 dyn_Cloth_volume_r_A01 [{shapeType=Capsule, length=5.834, radius=13.9863, shapeName=dyn_Cloth_volume_r_A01_capsule, center=<-17.2702 4.6454 0>}]
 dyn_Cloth_volume_r_A_00 [{shapeType=Capsule, length=7.8808, radius=12.4527, shapeName=dyn_Cloth_volume_r_A_capsule, center=<-19.8515 4.6454 0>}]
 dyn_Cloth_volume_r_B [{shapeType=Capsule, length=12.5884, radius=7.7451, shapeName=dyn_Cloth_volume_r_B_capsule, center=<-6.5604 -1.772 0>}]
 dyn_Cloth_volume_r_B01 [{shapeType=Capsule, length=10.9597, radius=11.3355, shapeName=dyn_Cloth_volume_r_B01_capsule, center=<-19.6493 -6.1459 0>}]
 dyn_Cloth_volume_r_B_00 [{shapeType=Capsule, length=12.5884, radius=7.7451, shapeName=dyn_Cloth_volume_r_B_capsule, center=<-15.5094 -1.772 0>}]
 dyn_Cloth_volume_r_C [{shapeType=Capsule, length=12.5884, radius=7.7451, shapeName=dyn_Cloth_volume_r_C_capsule, center=<-15.0543 6.68 0>}]
 dyn_Cloth_volume_r_C01 [{shapeType=Capsule, length=10.9597, radius=11.3355, shapeName=dyn_Cloth_volume_r_C01_capsule, center=<-20.6687 7.1556 0>}]
 dyn_Cloth_volume_r_C_00 [{shapeType=Capsule, length=12.9232, radius=9.8852, shapeName=dyn_Cloth_volume_r_C_capsule, center=<-19.0392 5.3658 0>}]
 pelvis [
  {shapeType=Capsule, length=20.4842, radius=18.4621, shapeName=pelvis_capsule, center=<0 4.2719 0>}
  {shapeType=Capsule, length=15.6605, radius=14.3395, shapeName=pelvis_capsule}
 ]
 spine_01 [{shapeType=Capsule, length=31.9929, radius=15.6577, shapeName=spine_01_capsule, center=<-7.7343 0 0>}]
 spine_02 [{shapeType=Capsule, length=19.677, radius=13.667, shapeName=spine_02_capsule, center=<0 2.4812 0>}]
 spine_04 [{shapeType=Capsule, length=19.677, radius=13.667, shapeName=spine_04_capsule, center=<0 0.341 0>}]
 thigh_l [{
   shapeType=Capsule
   length=42.5847
   radius=7.8129
   shapeName=thigh_l_capsule
   center=<-17.3631 0.67 -0.5841>
   rotation=<rot -88.0732 -0.0002 0.0002>
  }]
 thigh_r [{
   shapeType=Capsule
   length=43.4458
   radius=6.9518
   shapeName=thigh_r_capsule
   center=<19.1593 0 -0.5557>
   rotation=<rot -88.0732 -0.0002 0.0002>
  }]
}
constraints [
 {bone1Name=dyn_Cloth_volume_bk, bone2Name=pelvis, swing1LimitDegrees=45, swing2LimitDegrees=45, twistLimitDegrees=10}
 {bone1Name=dyn_Cloth_volume_bk01, bone2Name=dyn_Cloth_volume_bk_00, swing1LimitDegrees=45, swing2LimitDegrees=45, twistLimitDegrees=10}
 {bone1Name=dyn_Cloth_volume_fn, bone2Name=pelvis, swing1LimitDegrees=45, swing2LimitDegrees=45, twistLimitDegrees=10}
 {bone1Name=dyn_Cloth_volume_fn01, bone2Name=dyn_Cloth_volume_fn_00, swing1LimitDegrees=45, swing2LimitDegrees=45, twistLimitDegrees=10}
 {bone1Name=dyn_Cloth_volume_l_A, bone2Name=pelvis, swing1LimitDegrees=45, swing2LimitDegrees=45, twistLimitDegrees=10}
 {bone1Name=dyn_Cloth_volume_l_A01, bone2Name=dyn_Cloth_volume_l_A_00, swing1LimitDegrees=45, swing2LimitDegrees=45, twistLimitDegrees=10}
 {bone1Name=dyn_Cloth_volume_l_B, bone2Name=pelvis, swing1LimitDegrees=45, swing2LimitDegrees=45, twistLimitDegrees=10}
 {bone1Name=dyn_Cloth_volume_l_B01, bone2Name=dyn_Cloth_volume_l_B_00, swing1LimitDegrees=45, swing2LimitDegrees=45, twistLimitDegrees=10}
 {bone1Name=dyn_Cloth_volume_l_C, bone2Name=pelvis, swing1LimitDegrees=45, swing2LimitDegrees=45, twistLimitDegrees=10}
 {bone1Name=dyn_Cloth_volume_l_C01, bone2Name=dyn_Cloth_volume_l_C_00, swing1LimitDegrees=45, swing2LimitDegrees=45, twistLimitDegrees=10}
 {bone1Name=dyn_Cloth_volume_r_A, bone2Name=pelvis, swing1LimitDegrees=45, swing2LimitDegrees=45, twistLimitDegrees=10}
 {bone1Name=dyn_Cloth_volume_r_A01, bone2Name=dyn_Cloth_volume_r_A_00, swing1LimitDegrees=45, swing2LimitDegrees=45, twistLimitDegrees=10}
 {bone1Name=dyn_Cloth_volume_r_B, bone2Name=pelvis, swing1LimitDegrees=45, swing2LimitDegrees=45, twistLimitDegrees=10}
 {bone1Name=dyn_Cloth_volume_r_B01, bone2Name=dyn_Cloth_volume_r_B_00, swing1LimitDegrees=45, swing2LimitDegrees=45, twistLimitDegrees=10}
 {bone1Name=dyn_Cloth_volume_r_C, bone2Name=pelvis, swing1LimitDegrees=45, swing2LimitDegrees=45, twistLimitDegrees=10}
 {bone1Name=dyn_Cloth_volume_r_C01, bone2Name=dyn_Cloth_volume_r_C_00, swing1LimitDegrees=45, swing2LimitDegrees=45, twistLimitDegrees=10}
 {bone1Name=dyn_Cloth_volume_fn_00, bone2Name=dyn_Cloth_volume_fn, swing1LimitDegrees=45, swing2LimitDegrees=45, twistLimitDegrees=10}
 {bone1Name=dyn_Cloth_volume_bk_00, bone2Name=dyn_Cloth_volume_bk, swing1LimitDegrees=45, swing2LimitDegrees=45, twistLimitDegrees=10}
 {bone1Name=dyn_Cloth_volume_l_A_00, bone2Name=dyn_Cloth_volume_l_A, swing1LimitDegrees=45, swing2LimitDegrees=45, twistLimitDegrees=10}
 {bone1Name=dyn_Cloth_volume_l_B_00, bone2Name=dyn_Cloth_volume_l_B, swing1LimitDegrees=45, swing2LimitDegrees=45, twistLimitDegrees=10}
 {bone1Name=dyn_Cloth_volume_l_C_00, bone2Name=dyn_Cloth_volume_l_C, swing1LimitDegrees=45, swing2LimitDegrees=45, twistLimitDegrees=10}
 {bone1Name=dyn_Cloth_volume_r_A_00, bone2Name=dyn_Cloth_volume_r_A, swing1LimitDegrees=45, swing2LimitDegrees=45, twistLimitDegrees=10}
 {bone1Name=dyn_Cloth_volume_r_B_00, bone2Name=dyn_Cloth_volume_r_B, swing1LimitDegrees=45, swing2LimitDegrees=45, twistLimitDegrees=10}
 {bone1Name=dyn_Cloth_volume_r_C_00, bone2Name=dyn_Cloth_volume_r_C, swing1LimitDegrees=45, swing2LimitDegrees=45, twistLimitDegrees=10}
 {bone1Name=dyn_Cloth_volume_l_B_00, bone2Name=dyn_Cloth_volume_l_C_00, swing1LimitDegrees=45, swing2LimitDegrees=45, twistLimitDegrees=45}
 {bone1Name=dyn_Cloth_volume_r_B_00, bone2Name=dyn_Cloth_volume_r_C_00, swing1LimitDegrees=45, swing2LimitDegrees=45, twistLimitDegrees=45}
 {bone1Name=dyn_Cloth_volume_r_A_00, bone2Name=dyn_Cloth_volume_r_B_00, swing1LimitDegrees=45, swing2LimitDegrees=45, twistLimitDegrees=45}
 {bone1Name=dyn_Cloth_volume_l_A_00, bone2Name=dyn_Cloth_volume_l_B_00, swing1LimitDegrees=45, swing2LimitDegrees=45, twistLimitDegrees=45}
 {bone1Name=dyn_Cloth_volume_l_C_spine, bone2Name=dyn_Cloth_volume_l_C_00, swing1LimitDegrees=45, swing2LimitDegrees=45, twistLimitDegrees=6}
 {
  bone1Name=dyn_Cloth_volume_l_C01
  bone2Name=dyn_Cloth_volume_bk01
  swing1LimitDegrees=45
  swing1Motion=Free
  swing2LimitDegrees=45
  swing2Motion=Free
  twistLimitDegrees=45
  twistMotion=Free
 }
 {
  bone1Name=dyn_Cloth_volume_r_C01
  bone2Name=dyn_Cloth_volume_bk01
  swing1LimitDegrees=45
  swing1Motion=Free
  swing2LimitDegrees=45
  swing2Motion=Free
  twistLimitDegrees=45
  twistMotion=Free
 }
]
