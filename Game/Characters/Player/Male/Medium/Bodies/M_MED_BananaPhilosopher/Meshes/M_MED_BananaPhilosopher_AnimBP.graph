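# AnimBlueprint for the BananaPhilosopher body that copies the attached parent's pose and adds secondary motion to its skirt-like cloth and brooch.
# A skirt Control Rig, weighted constraints and corrective bone offsets build a cached base pose.
# Per-bone AnimDynamics and two RigidBody passes build the dynamic pose, with blends for swimming, DBNO and dive states plus sliding offsets.
# The event graph derives corrective alphas from body-mesh bone rotations.
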AnimBlueprint /Game/Characters/Player/Male/Medium/Bodies/M_MED_BananaPhilosopher/Meshes/M_MED_BananaPhilosopher_AnimBP
target_skeleton=&/M_MED_BananaPhilosopher_Skeleton
graphs [
 AnimGraph {nodes [
   n1 = Anim Dynamics - Bone: dyn_Cloth_volume_l_A_00!(ComponentPose, Alpha=1, GravityOverride:Vector="0,0,500", ExternalForce="0,0,0") -> (Pose) {
    angularDampingOverride=0.99
    angularSpringConstant=800
    bAngularSpring=true
    bOverrideAngularDamping=true
    bUseGravityOverride=true
    bUsePlanarLimit=false
    gravityScale=0.3
    simulationSpace=BoneRelative
    alphaBoolBlend {blendOption=Linear}
    boundBone {boneName=dyn_Cloth_volume_l_A_00}
    componentAppliedLinearAccClamp=<2000 2000 1600>
    componentLinearAccScale=<0.45 0.45 0.45>
    componentLinearVelScale=<0.55 0.55 0.55>
    gravityOverride {z=500}
    physicsBodyDefinitions [{
      collisionType=CustomSphere
      sphereCollisionRadius=2
      boundBone {boneName=dyn_Cloth_volume_l_A_00}
      boxExtents=<30 30 30>
      constraintSetup {angularLimitsMax=<0 0 5>, angularLimitsMin=<0 0 0>, angularTarget=<0 0 0>}
      localJointOffset=<-5 0.2 1>
     }]
    relativeSpaceBone {boneName=root}
   }
   n2 = Anim Dynamics - Bone: dyn_Cloth_volume_l_A01!(ComponentPose=@n1, Alpha=1, GravityOverride:Vector="0,0,500", ExternalForce="0,0,0") -> (Pose) {
    angularDampingOverride=0.99
    angularSpringConstant=800
    bAngularSpring=true
    bOverrideAngularDamping=true
    bUseGravityOverride=true
    bUsePlanarLimit=false
    gravityScale=0.3
    simulationSpace=BoneRelative
    alphaBoolBlend {blendOption=Linear}
    boundBone {boneName=dyn_Cloth_volume_l_A01}
    componentAppliedLinearAccClamp=<2000 2000 1600>
    componentLinearAccScale=<0.45 0.45 0.45>
    componentLinearVelScale=<0.55 0.55 0.55>
    gravityOverride {z=500}
    physicsBodyDefinitions [{
      collisionType=CustomSphere
      sphereCollisionRadius=2
      boundBone {boneName=dyn_Cloth_volume_l_A01}
      boxExtents=<20 20 20>
      constraintSetup {angularLimitsMax=<0 0 5>, angularLimitsMin=<0 0 0>, angularTarget=<0 0 0>}
      localJointOffset=<-5 0.2 1>
     }]
    relativeSpaceBone {boneName=root}
   }
   Anim Dynamics - Bone: dyn_Cloth_volume_l_A02!(ComponentPose=@n2, Alpha=1, GravityOverride:Vector="0,0,500", ExternalForce="0,0,0") -> (Pose) {
    angularDampingOverride=0.99
    angularSpringConstant=800
    bAngularSpring=true
    bOverrideAngularDamping=true
    bUseGravityOverride=true
    bUsePlanarLimit=false
    gravityScale=0.3
    simulationSpace=BoneRelative
    alphaBoolBlend {blendOption=Linear}
    boundBone {boneName=dyn_Cloth_volume_l_A02}
    componentAppliedLinearAccClamp=<2000 2000 1600>
    componentLinearAccScale=<0.45 0.45 0.45>
    componentLinearVelScale=<0.55 0.55 0.55>
    gravityOverride {z=500}
    physicsBodyDefinitions [{
      collisionType=CustomSphere
      sphereCollisionRadius=2
      boundBone {boneName=dyn_Cloth_volume_l_A02}
      boxExtents=<20 20 20>
      constraintSetup {angularLimitsMax=<0 0 5>, angularLimitsMin=<0 0 0>, angularTarget=<0 0 0>}
      localJointOffset=<-5 0.2 1>
     }]
    relativeSpaceBone {boneName=root}
   }
   n3 = "Transform (Modify) Bone - Bone: dyn_Cloth_volume_l_C"(
    ComponentPose=Local To Component(LocalPose=Use cached pose 'Is DynNO/OFF_Pose'())
    Alpha=1
    Translation="25,-35,-10"
    Rotation="0,0,0"
    Scale="1,1,1"
   ) {
    rotationMode=BMM_Additive
    rotationSpace=BCS_BoneSpace
    scaleMode=BMM_Additive
    scaleSpace=BCS_BoneSpace
    translationMode=BMM_Additive
    translationSpace=BCS_BoneSpace
    alphaBoolBlend {blendOption=Linear}
    boneToModify {boneName=dyn_Cloth_volume_l_C}
    translation=<25 -35 -10>
   }
   n4 = "Transform (Modify) Bone - Bone: dyn_Cloth_volume_r_C"(ComponentPose=@n3, Alpha=1, Translation="25,-25,0", Rotation="0,0,0", Scale="1,1,1") {
    rotationMode=BMM_Additive
    rotationSpace=BCS_BoneSpace
    scaleMode=BMM_Additive
    scaleSpace=BCS_BoneSpace
    translationMode=BMM_Additive
    translationSpace=BCS_BoneSpace
    alphaBoolBlend {blendOption=Linear}
    boneToModify {boneName=dyn_Cloth_volume_r_C}
    translation=<25 -25>
   }
   n5 = "Transform (Modify) Bone - Bone: dyn_Cloth_volume_bk"(ComponentPose=@n4, Alpha=1, Translation="25,-35,0", Rotation="0,0,0", Scale="1,1,1") {
    rotationMode=BMM_Additive
    rotationSpace=BCS_BoneSpace
    scaleMode=BMM_Additive
    scaleSpace=BCS_BoneSpace
    translationMode=BMM_Additive
    translationSpace=BCS_BoneSpace
    alphaBoolBlend {blendOption=Linear}
    boneToModify {boneName=dyn_Cloth_volume_bk}
    translation=<25 -35>
   }
   n6 = "Transform (Modify) Bone - Bone: dyn_Cloth_volume_l_C01"(ComponentPose=@n5, Alpha=1, Translation="10,20,0", Rotation="0,0,0", Scale="1,1,1") {
    rotationMode=BMM_Additive
    rotationSpace=BCS_BoneSpace
    scaleMode=BMM_Additive
    scaleSpace=BCS_BoneSpace
    translationMode=BMM_Additive
    translationSpace=BCS_BoneSpace
    alphaBoolBlend {blendOption=Linear}
    boneToModify {boneName=dyn_Cloth_volume_l_C01}
    translation=<10 20>
   }
   n7 = "Transform (Modify) Bone - Bone: dyn_Cloth_volume_r_C01"(ComponentPose=@n6, Alpha=1, Translation="10,20,0", Rotation="0,0,0", Scale="1,1,1") {
    rotationMode=BMM_Additive
    rotationSpace=BCS_BoneSpace
    scaleMode=BMM_Additive
    scaleSpace=BCS_BoneSpace
    translationMode=BMM_Additive
    translationSpace=BCS_BoneSpace
    alphaBoolBlend {blendOption=Linear}
    boneToModify {boneName=dyn_Cloth_volume_r_C01}
    translation=<10 20>
   }
   n8 = "Transform (Modify) Bone - Bone: dyn_Cloth_volume_bk01"(ComponentPose=@n7, Alpha=1, Translation="10,20,0", Rotation="0,0,0", Scale="1,1,1") {
    rotationMode=BMM_Additive
    rotationSpace=BCS_BoneSpace
    scaleMode=BMM_Additive
    scaleSpace=BCS_BoneSpace
    translationMode=BMM_Additive
    translationSpace=BCS_BoneSpace
    alphaBoolBlend {blendOption=Linear}
    boneToModify {boneName=dyn_Cloth_volume_bk01}
    translation=<10 20>
   }
   Output Pose(Result=Blend Poses by bool(
     BlendPose_0=Component To Local(ComponentPose=@n8)
     BlendPose_1=Use cached pose 'Is DynNO/OFF_Pose'()
     BlendTime_0=0.5
     bActiveValue=Get Is KneeAndBaseball Sliding() -> (Is KneeAndBaseball Sliding:bool)
    ) {blendType=Linear}) {layerGroup=None}
   n9 = Use cached pose 'SkirtCachePose'()
   n10 = Use cached pose 'SkirtDynPose'()
   n11 = Get DBNO() -> (DBNO:double)
   n12 = "Constraint - dyn_Cloth_volume_r_C01 (RR)"!(
    ComponentPose=Local To Component(LocalPose=Two Way Blend(A=@n10, B=@n9, Alpha=@n11) {blendNode {alphaBoolBlend {blendOption=Linear}}})
    Alpha=1
    ConstraintWeights_0=0.25
    ConstraintWeights_1:float=0.75
   ) -> (Pose) {
    alphaBoolBlend {blendOption=Linear}
    boneToModify {boneName=dyn_Cloth_volume_r_C01}
    constraintSetup [{transformType=Rotation, targetBone {boneName=thigh_r}}, {transformType=Rotation, targetBone {boneName=dyn_Cloth_volume_r_C01}}]
    constraintWeights [1,1]
   }
   n13 = "Constraint - dyn_Cloth_volume_l_C01 (RR)"!(ComponentPose=@n12, Alpha=1, ConstraintWeights_0=0.25, ConstraintWeights_1:float=0.75) -> (Pose) {
    alphaBoolBlend {blendOption=Linear}
    boneToModify {boneName=dyn_Cloth_volume_l_C01}
    constraintSetup [{transformType=Rotation, targetBone {boneName=thigh_l}}, {transformType=Rotation, targetBone {boneName=dyn_Cloth_volume_l_C01}}]
    constraintWeights [1,1]
   }
   n14 = "Constraint - dyn_Cloth_volume_r_B01 (PP)"!(ComponentPose=@n13, Alpha=1, ConstraintWeights_0=0.2, ConstraintWeights_1:float=0.8) -> (Pose) {
    alphaBoolBlend {blendOption=Linear}
    boneToModify {boneName=dyn_Cloth_volume_r_B01}
    constraintSetup [{targetBone {boneName=dyn_Cloth_volume_r_C01}}, {targetBone {boneName=dyn_Cloth_volume_r_B01}}]
    constraintWeights [1,1]
   }
   n15 = "Constraint - dyn_Cloth_volume_l_B01 (PP)"!(ComponentPose=@n14, Alpha=1, ConstraintWeights_0=0.2, ConstraintWeights_1:float=0.8) -> (Pose) {
    alphaBoolBlend {blendOption=Linear}
    boneToModify {boneName=dyn_Cloth_volume_l_B01}
    constraintSetup [{targetBone {boneName=dyn_Cloth_volume_l_C01}}, {targetBone {boneName=dyn_Cloth_volume_l_B01}}]
    constraintWeights [1,1]
   }
   n16 = "Constraint - dyn_Cloth_volume_bk01 (RRR)"!(ComponentPose=@n15, Alpha=1, ConstraintWeights_0=0.15, ConstraintWeights_1:float=0.15, ConstraintWeights_2:float=0.7) -> (Pose) {
    alphaBoolBlend {blendOption=Linear}
    boneToModify {boneName=dyn_Cloth_volume_bk01}
    constraintSetup [
     {transformType=Rotation, targetBone {boneName=thigh_l}}
     {transformType=Rotation, targetBone {boneName=thigh_r}}
     {transformType=Rotation, targetBone {boneName=dyn_Cloth_volume_bk01}}
    ]
    constraintWeights [1,1,1]
   }
   n17 = Component To Local(ComponentPose=@n16)
   n18 = Use cached pose 'SkirtDynPose'()
   n19 = Get Is Dive And DBNO() -> (Is Dive And DBNO:bool)
   Save cached pose 'Is DynNO/OFF_Pose'(Pose=Blend Poses by bool(
     BlendPose_0=Blend Poses by bool(BlendPose_0=@n17, BlendPose_1=@n18, bActiveValue=@n19) {blendType=Linear}
     BlendPose_1=Use cached pose 'SkirtCachePose'()
    ) {blendType=Linear}) {cacheName=Is DynNO/OFF_Pose}
   n20 = Control Rig(Source=Copy Pose From Mesh() {bUseAttachedParent=true}) {
    alphaBoolBlend {blendOption=Linear}
    controlRigAssetReference {blueprintRigClass=&/M_MED_BananaPhilosopher_Skirt_CtrlRig.M_MED_BananaPhilosopher_Skirt_CtrlRig_C}
   }
   n21 = "Constraint - attach_cape (PPPP)"!(
    ComponentPose=Local To Component(LocalPose=@n20)
    Alpha=1
    ConstraintWeights_0=0.7815
    ConstraintWeights_1:float=0.1023
    ConstraintWeights_2:float=0.0909
    ConstraintWeights_3:float=0.0254
   ) -> (Pose) {
    alphaBoolBlend {blendOption=Linear}
    boneToModify {boneName=attach_cape}
    constraintSetup [{targetBone {boneName=spine_05}}, {targetBone {boneName=spine_04}}, {targetBone {boneName=spine_03}}, {targetBone {boneName=neck_01}}]
    constraintWeights [1,1,1,1]
   }
   n22 = "Constraint - attach_backpack (PPPPP)"!(
    ComponentPose=@n21
    Alpha=1
    ConstraintWeights_0=0.4832
    ConstraintWeights_1:float=0.2884
    ConstraintWeights_2:float=0.1965
    ConstraintWeights_3:float=0.026
    ConstraintWeights_4:float=0.006
   ) -> (Pose) {
    alphaBoolBlend {blendOption=Linear}
    boneToModify {boneName=attach_backpack}
    constraintSetup [
     {targetBone {boneName=spine_05}}
     {targetBone {boneName=spine_04}}
     {targetBone {boneName=spine_03}}
     {targetBone {boneName=spine_02}}
     {targetBone {boneName=pelvis}}
    ]
    constraintWeights [1,1,1,1,1]
   }
   n23 = "Constraint - gravity (TT)"!(ComponentPose=@n22, Alpha=1, ConstraintWeights_0=0, ConstraintWeights_1:float=1) -> (Pose) {
    alphaBoolBlend {blendOption=Linear}
    boneToModify {boneName=gravity}
    constraintSetup [{transformType=Translation, targetBone {boneName=root}}, {transformType=Translation, targetBone {boneName=pelvis}}]
    constraintWeights [1,1]
   }
   n24 = "Constraint - gravity (RR)"!(ComponentPose=@n23, Alpha=1, ConstraintWeights_0=0, ConstraintWeights_1:float=1) -> (Pose) {
    alphaBoolBlend {blendOption=Linear}
    boneToModify {boneName=gravity}
    constraintSetup [{transformType=Rotation, targetBone {boneName=root}}, {transformType=Rotation, targetBone {boneName=pelvis}}]
    constraintWeights [1,1]
   }
   n25 = "Constraint - dyn_Cloth_spine_00 (PPPPPP)"!(
    ComponentPose=@n24
    Alpha=1
    ConstraintWeights_0=0.4081
    ConstraintWeights_1:float=0.171
    ConstraintWeights_2:float=0.1708
    ConstraintWeights_3:float=0.1405
    ConstraintWeights_4:float=0.0785
    ConstraintWeights_5:float=0.031
   ) -> (Pose) {
    alphaBoolBlend {blendOption=Linear}
    boneToModify {boneName=dyn_Cloth_spine_00}
    constraintSetup [
     {targetBone {boneName=spine_04}}
     {targetBone {boneName=clavicle_l}}
     {targetBone {boneName=spine_05}}
     {targetBone {boneName=spine_03}}
     {targetBone {boneName=upperarm_twist_01_l}}
     {targetBone {boneName=spine_02}}
    ]
    constraintWeights [1,1,1,1,1,1]
   }
   n26 = "Constraint - dyn_Cloth_spine_01 (RRR)"!(ComponentPose=@n25, Alpha=1, ConstraintWeights_0=0.2, ConstraintWeights_1:float=0.8, ConstraintWeights_2:float=0) -> (Pose) {
    alphaBoolBlend {blendOption=Linear}
    boneToModify {boneName=dyn_Cloth_spine_01}
    constraintSetup [
     {transformType=Rotation, targetBone {boneName=dyn_Cloth_volume_fn}}
     {transformType=Rotation, targetBone {boneName=dyn_Cloth_spine_00}}
     {transformType=Rotation, targetBone {boneName=thigh_twist_01_l}}
    ]
    constraintWeights [1,1,1]
   }
   n27 = "Constraint - dyn_Cloth_spine_02 (RRR)"!(ComponentPose=@n26, Alpha=1, ConstraintWeights_0=0, ConstraintWeights_1:float=0.6, ConstraintWeights_2:float=0.4) -> (Pose) {
    alphaBoolBlend {blendOption=Linear}
    boneToModify {boneName=dyn_Cloth_spine_02}
    constraintSetup [
     {transformType=Rotation, targetBone {boneName=dyn_Cloth_volume_l_A}}
     {transformType=Rotation, targetBone {boneName=dyn_Cloth_spine_02}}
     {transformType=Rotation, targetBone {boneName=thigh_twist_01_l}}
    ]
    constraintWeights [1,1,1]
   }
   n28 = "Constraint - dyn_Cloth_spine_03 (RRR)"!(ComponentPose=@n27, Alpha=1, ConstraintWeights_0=0, ConstraintWeights_1:float=0, ConstraintWeights_2:float=1) -> (Pose) {
    alphaBoolBlend {blendOption=Linear}
    boneToModify {boneName=dyn_Cloth_spine_03}
    constraintSetup [
     {transformType=Rotation, targetBone {boneName=dyn_Cloth_volume_l_A}}
     {transformType=Rotation, targetBone {boneName=dyn_Cloth_spine_02}}
     {transformType=Rotation, targetBone {boneName=thigh_twist_01_l}}
    ]
    constraintWeights [1,1,1]
   }
   n29 = "Constraint - dyn_Cloth_volume_bk (PP)"!(ComponentPose=@n28, Alpha=1, ConstraintWeights_0=0.5, ConstraintWeights_1:float=0.5) -> (Pose) {
    alphaBoolBlend {blendOption=Linear}
    boneToModify {boneName=dyn_Cloth_volume_bk}
    constraintSetup [{targetBone {boneName=dyn_Cloth_volume_l_C}}, {targetBone {boneName=dyn_Cloth_volume_r_C}}]
    constraintWeights [1,1]
   }
   n30 = "Constraint - dyn_Cloth_volume_bk01 (PP)"!(ComponentPose=@n29, Alpha=1, ConstraintWeights_0=0.5, ConstraintWeights_1:float=0.5) -> (Pose) {
    alphaBoolBlend {blendOption=Linear}
    boneToModify {boneName=dyn_Cloth_volume_bk01}
    constraintSetup [{targetBone {boneName=dyn_Cloth_volume_l_C01}}, {targetBone {boneName=dyn_Cloth_volume_r_C01}}]
    constraintWeights [1,1]
   }
   n31 = "Constraint - dyn_Cloth_volume_bk02 (PP)"!(ComponentPose=@n30, Alpha=1, ConstraintWeights_0=0.5, ConstraintWeights_1:float=0.5) -> (Pose) {
    alphaBoolBlend {blendOption=Linear}
    boneToModify {boneName=dyn_Cloth_volume_bk02}
    constraintSetup [{targetBone {boneName=dyn_Cloth_volume_l_C02}}, {targetBone {boneName=dyn_Cloth_volume_r_C02}}]
    constraintWeights [1,1]
   }
   n32 = Get thigh_shape_fn_l() -> (thigh_shape_fn_l:double)
   n33 = "Transform (Modify) Bone - Bone: dyn_Cloth_spine_00"(ComponentPose=@n31, Alpha=@n32, Translation="0,0,0", Rotation="-50,20,50", Scale="1,1,1") {
    rotationMode=BMM_Additive
    rotationSpace=BCS_BoneSpace
    scaleMode=BMM_Additive
    scaleSpace=BCS_BoneSpace
    translationMode=BMM_Additive
    translationSpace=BCS_BoneSpace
    alphaBoolBlend {blendOption=Linear}
    boneToModify {boneName=dyn_Cloth_spine_00}
    rotation=<rot -50 20 50>
   }
   n34 = "Transform (Modify) Bone - Bone: dyn_Cloth_spine_01"(ComponentPose=@n33, Alpha=@n32, Translation="0,0,0", Rotation="-50,-20,0", Scale="1,1,1") {
    rotationMode=BMM_Additive
    rotationSpace=BCS_BoneSpace
    scaleMode=BMM_Additive
    scaleSpace=BCS_BoneSpace
    translationMode=BMM_Additive
    translationSpace=BCS_BoneSpace
    alphaBoolBlend {blendOption=Linear}
    boneToModify {boneName=dyn_Cloth_spine_01}
    rotation {pitch=-50, yaw=-20}
   }
   n35 = "Transform (Modify) Bone - Bone: dyn_Cloth_spine_02"(ComponentPose=@n34, Alpha=@n32, Translation="0,0,0", Rotation="30,-20,0", Scale="1,1,1") {
    rotationMode=BMM_Additive
    rotationSpace=BCS_BoneSpace
    scaleMode=BMM_Additive
    scaleSpace=BCS_BoneSpace
    translationMode=BMM_Additive
    translationSpace=BCS_BoneSpace
    alphaBoolBlend {blendOption=Linear}
    boneToModify {boneName=dyn_Cloth_spine_02}
    rotation {pitch=30, yaw=-20}
   }
   Save cached pose 'SkirtCachePose'(Pose=Component To Local(ComponentPose=@n35)) {cacheName=SkirtCachePose}
   n36 = Anim Dynamics - Bone: dyn_Cloth_volume_fn_00!(
    ComponentPose=Local To Component(LocalPose=Use cached pose 'SkirtCachePose'())
    Alpha=1
    GravityOverride:Vector="0,0,500"
    ExternalForce="0,0,0"
   ) -> (Pose) {
    angularDampingOverride=0.99
    angularSpringConstant=800
    bAngularSpring=true
    bOverrideAngularDamping=true
    bUseGravityOverride=true
    bUsePlanarLimit=false
    gravityScale=0.3
    simulationSpace=BoneRelative
    alphaBoolBlend {blendOption=Linear}
    boundBone {boneName=dyn_Cloth_volume_fn_00}
    componentAppliedLinearAccClamp=<2000 2000 1600>
    componentLinearAccScale=<0.45 0.45 0.45>
    componentLinearVelScale=<0.55 0.55 0.55>
    gravityOverride {z=500}
    physicsBodyDefinitions [{
      collisionType=CustomSphere
      sphereCollisionRadius=2
      boundBone {boneName=dyn_Cloth_volume_fn_00}
      boxExtents=<30 30 30>
      constraintSetup {angularLimitsMax=<0 0 5>, angularLimitsMin=<0 0 0>, angularTarget=<0 0 0>}
      localJointOffset=<-5 0.2 1>
     }]
    relativeSpaceBone {boneName=root}
   }
   n37 = Anim Dynamics - Bone: dyn_Cloth_volume_fn01!(ComponentPose=@n36, Alpha=1, GravityOverride:Vector="0,0,500", ExternalForce="0,0,0") -> (Pose) {
    angularDampingOverride=0.99
    angularSpringConstant=800
    bAngularSpring=true
    bOverrideAngularDamping=true
    bUseGravityOverride=true
    bUsePlanarLimit=false
    gravityScale=0.3
    simulationSpace=BoneRelative
    alphaBoolBlend {blendOption=Linear}
    boundBone {boneName=dyn_Cloth_volume_fn01}
    componentAppliedLinearAccClamp=<2000 2000 1600>
    componentLinearAccScale=<0.45 0.45 0.45>
    componentLinearVelScale=<0.55 0.55 0.55>
    gravityOverride {z=500}
    physicsBodyDefinitions [{
      collisionType=CustomSphere
      sphereCollisionRadius=2
      boundBone {boneName=dyn_Cloth_volume_fn01}
      boxExtents=<20 20 20>
      constraintSetup {angularLimitsMax=<0 0 5>, angularLimitsMin=<0 0 0>, angularTarget=<0 0 0>}
      localJointOffset=<-5 0.2 1>
     }]
    relativeSpaceBone {boneName=root}
   }
   n38 = Anim Dynamics - Bone: dyn_Cloth_volume_fn02!(ComponentPose=@n37, Alpha=1, GravityOverride:Vector="0,0,500", ExternalForce="0,0,0") -> (Pose) {
    angularDampingOverride=0.99
    angularSpringConstant=800
    bAngularSpring=true
    bOverrideAngularDamping=true
    bUseGravityOverride=true
    bUsePlanarLimit=false
    gravityScale=0.3
    simulationSpace=BoneRelative
    alphaBoolBlend {blendOption=Linear}
    boundBone {boneName=dyn_Cloth_volume_fn02}
    componentAppliedLinearAccClamp=<2000 2000 1600>
    componentLinearAccScale=<0.45 0.45 0.45>
    componentLinearVelScale=<0.55 0.55 0.55>
    gravityOverride {z=500}
    physicsBodyDefinitions [{
      collisionType=CustomSphere
      sphereCollisionRadius=2
      boundBone {boneName=dyn_Cloth_volume_fn02}
      boxExtents=<20 20 20>
      constraintSetup {angularLimitsMax=<0 0 5>, angularLimitsMin=<0 0 0>, angularTarget=<0 0 0>}
      localJointOffset=<-5 0.2 1>
     }]
    relativeSpaceBone {boneName=root}
   }
   n39 = Anim Dynamics - Bone: dyn_Cloth_volume_l_B_00!(ComponentPose=@n38, Alpha=1, GravityOverride:Vector="0,0,500", ExternalForce="0,0,0") -> (Pose) {
    angularDampingOverride=0.99
    angularSpringConstant=800
    bAngularSpring=true
    bOverrideAngularDamping=true
    bUseGravityOverride=true
    bUsePlanarLimit=false
    gravityScale=0.3
    simulationSpace=BoneRelative
    alphaBoolBlend {blendOption=Linear}
    boundBone {boneName=dyn_Cloth_volume_l_B_00}
    componentAppliedLinearAccClamp=<2000 2000 1600>
    componentLinearAccScale=<0.4 0.4 0.4>
    componentLinearVelScale=<0.45 0.45 0.55>
    gravityOverride {z=500}
    physicsBodyDefinitions [{
      collisionType=CustomSphere
      sphereCollisionRadius=2
      boundBone {boneName=dyn_Cloth_volume_l_B_00}
      boxExtents=<50 50 50>
      constraintSetup {angularLimitsMax=<0 0 5>, angularLimitsMin=<0 0 0>, angularTarget=<0 0 0>}
      localJointOffset=<-5 0.2 -0.1>
     }]
    relativeSpaceBone {boneName=root}
   }
   n40 = Anim Dynamics - Bone: dyn_Cloth_volume_l_B01!(ComponentPose=@n39, Alpha=1, GravityOverride:Vector="0,0,500", ExternalForce="0,0,0") -> (Pose) {
    angularDampingOverride=0.99
    angularSpringConstant=800
    bAngularSpring=true
    bOverrideAngularDamping=true
    bUseGravityOverride=true
    bUsePlanarLimit=false
    gravityScale=0.3
    simulationSpace=BoneRelative
    alphaBoolBlend {blendOption=Linear}
    boundBone {boneName=dyn_Cloth_volume_l_B01}
    componentAppliedLinearAccClamp=<2000 2000 1600>
    componentLinearAccScale=<0.4 0.4 0.4>
    componentLinearVelScale=<0.45 0.45 0.55>
    gravityOverride {z=500}
    physicsBodyDefinitions [{
      collisionType=CustomSphere
      sphereCollisionRadius=2
      boundBone {boneName=dyn_Cloth_volume_l_B01}
      boxExtents=<33 33 33>
      constraintSetup {angularLimitsMax=<0 0 5>, angularLimitsMin=<0 0 0>, angularTarget=<0 0 0>}
      localJointOffset=<-5 0.2 -0.1>
     }]
    relativeSpaceBone {boneName=root}
   }
   n41 = Anim Dynamics - Bone: dyn_Cloth_volume_l_B02!(ComponentPose=@n40, Alpha=1, GravityOverride:Vector="0,0,500", ExternalForce="0,0,0") -> (Pose) {
    angularDampingOverride=0.99
    angularSpringConstant=800
    bAngularSpring=true
    bOverrideAngularDamping=true
    bUseGravityOverride=true
    bUsePlanarLimit=false
    gravityScale=0.3
    simulationSpace=BoneRelative
    alphaBoolBlend {blendOption=Linear}
    boundBone {boneName=dyn_Cloth_volume_l_B02}
    componentAppliedLinearAccClamp=<2000 2000 1600>
    componentLinearAccScale=<0.4 0.4 0.4>
    componentLinearVelScale=<0.45 0.45 0.55>
    gravityOverride {z=500}
    physicsBodyDefinitions [{
      collisionType=CustomSphere
      sphereCollisionRadius=2
      boundBone {boneName=dyn_Cloth_volume_l_B02}
      boxExtents=<27 27 27>
      constraintSetup {angularLimitsMax=<0 0 5>, angularLimitsMin=<0 0 0>, angularTarget=<0 0 0>}
      localJointOffset=<-5 0.2 -0.1>
     }]
    relativeSpaceBone {boneName=root}
   }
   n42 = Anim Dynamics - Bone: dyn_Cloth_volume_l_C_00!(ComponentPose=@n41, Alpha=1, GravityOverride:Vector="0,0,500", ExternalForce="0,0,0") -> (Pose) {
    angularDampingOverride=0.99
    angularSpringConstant=800
    bAngularSpring=true
    bOverrideAngularDamping=true
    bUseGravityOverride=true
    bUsePlanarLimit=false
    gravityScale=0.3
    simulationSpace=BoneRelative
    alphaBoolBlend {blendOption=Linear}
    boundBone {boneName=dyn_Cloth_volume_l_C_00}
    componentAppliedLinearAccClamp=<2000 2000 1600>
    componentLinearAccScale=<0.5 0.5 0.4>
    componentLinearVelScale=<0.55 0.55 0.55>
    gravityOverride {z=500}
    physicsBodyDefinitions [{
      collisionType=CustomSphere
      sphereCollisionRadius=2
      boundBone {boneName=dyn_Cloth_volume_l_C_00}
      boxExtents=<60 60 60>
      constraintSetup {angularLimitsMax=<0 0 5>, angularLimitsMin=<0 0 0>, angularTarget=<0 0 0>}
      localJointOffset=<-5 0.2 -0.1>
     }]
    relativeSpaceBone {boneName=root}
   }
   n43 = Anim Dynamics - Bone: dyn_Cloth_volume_l_C_spine!(ComponentPose=@n42, Alpha=1, GravityOverride:Vector="0,0,500", ExternalForce="0,0,0") -> (Pose) {
    angularDampingOverride=0.99
    angularSpringConstant=800
    bAngularSpring=true
    bOverrideAngularDamping=true
    bUseGravityOverride=true
    bUsePlanarLimit=false
    gravityScale=0.3
    simulationSpace=BoneRelative
    alphaBoolBlend {blendOption=Linear}
    boundBone {boneName=dyn_Cloth_volume_l_C_spine}
    componentAppliedLinearAccClamp=<2000 2000 1600>
    componentLinearAccScale=<0.5 0.5 0.4>
    componentLinearVelScale=<0.55 0.55 0.55>
    gravityOverride {z=500}
    physicsBodyDefinitions [{
      collisionType=CustomSphere
      sphereCollisionRadius=2
      boundBone {boneName=dyn_Cloth_volume_l_C_spine}
      boxExtents=<30 30 30>
      constraintSetup {angularLimitsMax=<0 0 5>, angularLimitsMin=<0 0 0>, angularTarget=<0 0 0>}
      localJointOffset=<-5 0.2 -0.1>
     }]
    relativeSpaceBone {boneName=root}
   }
   n44 = Anim Dynamics - Bone: dyn_Cloth_volume_l_C01!(ComponentPose=@n43, Alpha=1, GravityOverride:Vector="0,0,500", ExternalForce="0,0,0") -> (Pose) {
    angularDampingOverride=0.99
    angularSpringConstant=800
    bAngularSpring=true
    bOverrideAngularDamping=true
    bUseGravityOverride=true
    bUsePlanarLimit=false
    gravityScale=0.3
    simulationSpace=BoneRelative
    alphaBoolBlend {blendOption=Linear}
    boundBone {boneName=dyn_Cloth_volume_l_C01}
    componentAppliedLinearAccClamp=<2000 2000 1600>
    componentLinearAccScale=<0.5 0.5 0.4>
    componentLinearVelScale=<0.55 0.55 0.55>
    gravityOverride {z=500}
    physicsBodyDefinitions [{
      collisionType=CustomSphere
      sphereCollisionRadius=2
      boundBone {boneName=dyn_Cloth_volume_l_C01}
      boxExtents=<50 50 50>
      constraintSetup {angularLimitsMax=<0 0 5>, angularLimitsMin=<0 0 0>, angularTarget=<0 0 0>}
      localJointOffset=<-5 0.2 -0.1>
     }]
    relativeSpaceBone {boneName=root}
   }
   n45 = Anim Dynamics - Bone: dyn_Cloth_volume_l_C02!(ComponentPose=@n44, Alpha=1, GravityOverride:Vector="0,0,500", ExternalForce="0,0,0") -> (Pose) {
    angularDampingOverride=0.99
    angularSpringConstant=800
    bAngularSpring=true
    bOverrideAngularDamping=true
    bUseGravityOverride=true
    bUsePlanarLimit=false
    gravityScale=0.3
    simulationSpace=BoneRelative
    alphaBoolBlend {blendOption=Linear}
    boundBone {boneName=dyn_Cloth_volume_l_C02}
    componentAppliedLinearAccClamp=<2000 2000 1600>
    componentLinearAccScale=<0.5 0.5 0.4>
    componentLinearVelScale=<0.55 0.55 0.55>
    gravityOverride {z=500}
    physicsBodyDefinitions [{
      collisionType=CustomSphere
      sphereCollisionRadius=2
      boundBone {boneName=dyn_Cloth_volume_l_C02}
      boxExtents=<40 40 40>
      constraintSetup {angularLimitsMax=<0 0 5>, angularLimitsMin=<0 0 0>, angularTarget=<0 0 0>}
      localJointOffset=<-5 0.2 -0.1>
     }]
    relativeSpaceBone {boneName=root}
   }
   n46 = Anim Dynamics - Bone: dyn_Cloth_volume_r_A_00!(ComponentPose=@n45, Alpha=1, GravityOverride:Vector="0,0,-980", ExternalForce="0,0,0") -> (Pose) {
    angularDampingOverride=0.99
    angularSpringConstant=800
    bAngularSpring=true
    bOverrideAngularDamping=true
    bUseGravityOverride=true
    bUsePlanarLimit=false
    gravityScale=0.3
    simulationSpace=BoneRelative
    alphaBoolBlend {blendOption=Linear}
    boundBone {boneName=dyn_Cloth_volume_r_A_00}
    componentAppliedLinearAccClamp=<2000 2000 1600>
    componentLinearAccScale=<0.45 0.45 0.45>
    componentLinearVelScale=<0.55 0.55 0.55>
    gravityOverride {z=-980}
    physicsBodyDefinitions [{
      collisionType=CustomSphere
      sphereCollisionRadius=2
      boundBone {boneName=dyn_Cloth_volume_r_A_00}
      boxExtents=<30 30 30>
      constraintSetup {angularLimitsMax=<0 0 5>, angularLimitsMin=<0 0 0>, angularTarget=<0 0 0>}
      localJointOffset=<-5 0.2 1>
     }]
    relativeSpaceBone {boneName=root}
   }
   n47 = Anim Dynamics - Bone: dyn_Cloth_volume_r_A01!(ComponentPose=@n46, Alpha=1, GravityOverride:Vector="0,0,-980", ExternalForce="0,0,0") -> (Pose) {
    angularDampingOverride=0.99
    angularSpringConstant=800
    bAngularSpring=true
    bOverrideAngularDamping=true
    bUseGravityOverride=true
    bUsePlanarLimit=false
    gravityScale=0.3
    simulationSpace=BoneRelative
    alphaBoolBlend {blendOption=Linear}
    boundBone {boneName=dyn_Cloth_volume_r_A01}
    componentAppliedLinearAccClamp=<2000 2000 1600>
    componentLinearAccScale=<0.45 0.45 0.45>
    componentLinearVelScale=<0.55 0.55 0.55>
    gravityOverride {z=-980}
    physicsBodyDefinitions [{
      collisionType=CustomSphere
      sphereCollisionRadius=2
      boundBone {boneName=dyn_Cloth_volume_r_A01}
      boxExtents=<20 20 20>
      constraintSetup {angularLimitsMax=<0 0 5>, angularLimitsMin=<0 0 0>, angularTarget=<0 0 0>}
      localJointOffset=<-5 0.2 1>
     }]
    relativeSpaceBone {boneName=root}
   }
   n48 = Anim Dynamics - Bone: dyn_Cloth_volume_r_A02!(ComponentPose=@n47, Alpha=1, GravityOverride:Vector="0,0,-980", ExternalForce="0,0,0") -> (Pose) {
    angularDampingOverride=0.99
    angularSpringConstant=800
    bAngularSpring=true
    bOverrideAngularDamping=true
    bUseGravityOverride=true
    bUsePlanarLimit=false
    gravityScale=0.3
    simulationSpace=BoneRelative
    alphaBoolBlend {blendOption=Linear}
    boundBone {boneName=dyn_Cloth_volume_r_A02}
    componentAppliedLinearAccClamp=<2000 2000 1600>
    componentLinearAccScale=<0.45 0.45 0.45>
    componentLinearVelScale=<0.55 0.55 0.55>
    gravityOverride {z=-980}
    physicsBodyDefinitions [{
      collisionType=CustomSphere
      sphereCollisionRadius=2
      boundBone {boneName=dyn_Cloth_volume_r_A02}
      boxExtents=<20 20 20>
      constraintSetup {angularLimitsMax=<0 0 5>, angularLimitsMin=<0 0 0>, angularTarget=<0 0 0>}
      localJointOffset=<-5 0.2 1>
     }]
    relativeSpaceBone {boneName=root}
   }
   n49 = Anim Dynamics - Bone: dyn_Cloth_volume_r_B_00!(ComponentPose=@n48, Alpha=1, GravityOverride:Vector="0,0,-980", ExternalForce="0,0,0") -> (Pose) {
    angularDampingOverride=0.99
    angularSpringConstant=800
    bAngularSpring=true
    bOverrideAngularDamping=true
    bUseGravityOverride=true
    bUsePlanarLimit=false
    gravityScale=0.3
    simulationSpace=BoneRelative
    alphaBoolBlend {blendOption=Linear}
    boundBone {boneName=dyn_Cloth_volume_r_B_00}
    componentAppliedLinearAccClamp=<2000 2000 1600>
    componentLinearAccScale=<0.4 0.4 0.4>
    componentLinearVelScale=<0.45 0.45 0.55>
    gravityOverride {z=-980}
    physicsBodyDefinitions [{
      collisionType=CustomSphere
      sphereCollisionRadius=2
      boundBone {boneName=dyn_Cloth_volume_r_B_00}
      boxExtents=<50 50 50>
      constraintSetup {angularLimitsMax=<0 0 5>, angularLimitsMin=<0 0 0>, angularTarget=<0 0 0>}
      localJointOffset=<-5 0.2 -0.1>
     }]
    relativeSpaceBone {boneName=root}
   }
   n50 = Anim Dynamics - Bone: dyn_Cloth_volume_r_B01!(ComponentPose=@n49, Alpha=1, GravityOverride:Vector="0,0,-980", ExternalForce="0,0,0") -> (Pose) {
    angularDampingOverride=0.99
    angularSpringConstant=800
    bAngularSpring=true
    bOverrideAngularDamping=true
    bUseGravityOverride=true
    bUsePlanarLimit=false
    gravityScale=0.3
    simulationSpace=BoneRelative
    alphaBoolBlend {blendOption=Linear}
    boundBone {boneName=dyn_Cloth_volume_r_B01}
    componentAppliedLinearAccClamp=<2000 2000 1600>
    componentLinearAccScale=<0.4 0.4 0.4>
    componentLinearVelScale=<0.45 0.45 0.55>
    gravityOverride {z=-980}
    physicsBodyDefinitions [{
      collisionType=CustomSphere
      sphereCollisionRadius=2
      boundBone {boneName=dyn_Cloth_volume_r_B01}
      boxExtents=<33 33 33>
      constraintSetup {angularLimitsMax=<0 0 5>, angularLimitsMin=<0 0 0>, angularTarget=<0 0 0>}
      localJointOffset=<-5 0.2 -0.1>
     }]
    relativeSpaceBone {boneName=root}
   }
   n51 = Anim Dynamics - Bone: dyn_Cloth_volume_r_B02!(ComponentPose=@n50, Alpha=1, GravityOverride:Vector="0,0,-980", ExternalForce="0,0,0") -> (Pose) {
    angularDampingOverride=0.99
    angularSpringConstant=800
    bAngularSpring=true
    bOverrideAngularDamping=true
    bUseGravityOverride=true
    bUsePlanarLimit=false
    gravityScale=0.3
    simulationSpace=BoneRelative
    alphaBoolBlend {blendOption=Linear}
    boundBone {boneName=dyn_Cloth_volume_r_B02}
    componentAppliedLinearAccClamp=<2000 2000 1600>
    componentLinearAccScale=<0.4 0.4 0.4>
    componentLinearVelScale=<0.45 0.45 0.55>
    gravityOverride {z=-980}
    physicsBodyDefinitions [{
      collisionType=CustomSphere
      sphereCollisionRadius=2
      boundBone {boneName=dyn_Cloth_volume_r_B02}
      boxExtents=<27 27 27>
      constraintSetup {angularLimitsMax=<0 0 5>, angularLimitsMin=<0 0 0>, angularTarget=<0 0 0>}
      localJointOffset=<-5 0.2 -0.1>
     }]
    relativeSpaceBone {boneName=root}
   }
   n52 = Anim Dynamics - Bone: dyn_Cloth_volume_r_C_00!(ComponentPose=@n51, Alpha=1, GravityOverride:Vector="0,0,-980", ExternalForce="0,0,0") -> (Pose) {
    angularDampingOverride=0.99
    angularSpringConstant=800
    bAngularSpring=true
    bOverrideAngularDamping=true
    bUseGravityOverride=true
    bUsePlanarLimit=false
    gravityScale=0.3
    simulationSpace=BoneRelative
    alphaBoolBlend {blendOption=Linear}
    boundBone {boneName=dyn_Cloth_volume_r_C_00}
    componentAppliedLinearAccClamp=<2000 2000 1600>
    componentLinearAccScale=<0.5 0.5 0.4>
    componentLinearVelScale=<0.55 0.55 0.55>
    gravityOverride {z=-980}
    physicsBodyDefinitions [{
      collisionType=CustomSphere
      sphereCollisionRadius=2
      boundBone {boneName=dyn_Cloth_volume_r_C_00}
      boxExtents=<45 45 45>
      constraintSetup {angularLimitsMax=<0 0 5>, angularLimitsMin=<0 0 0>, angularTarget=<0 0 0>}
      localJointOffset=<-5 0.2 -0.1>
     }]
    relativeSpaceBone {boneName=root}
   }
   n53 = Anim Dynamics - Bone: dyn_Cloth_volume_r_C01!(ComponentPose=@n52, Alpha=1, GravityOverride:Vector="0,0,-980", ExternalForce="0,0,0") -> (Pose) {
    angularDampingOverride=0.99
    angularSpringConstant=800
    bAngularSpring=true
    bOverrideAngularDamping=true
    bUseGravityOverride=true
    bUsePlanarLimit=false
    gravityScale=0.3
    simulationSpace=BoneRelative
    alphaBoolBlend {blendOption=Linear}
    boundBone {boneName=dyn_Cloth_volume_r_C01}
    componentAppliedLinearAccClamp=<2000 2000 1600>
    componentLinearAccScale=<0.5 0.5 0.4>
    componentLinearVelScale=<0.55 0.55 0.55>
    gravityOverride {z=-980}
    physicsBodyDefinitions [{
      collisionType=CustomSphere
      sphereCollisionRadius=2
      boundBone {boneName=dyn_Cloth_volume_r_C01}
      boxExtents=<40 40 40>
      constraintSetup {angularLimitsMax=<0 0 5>, angularLimitsMin=<0 0 0>, angularTarget=<0 0 0>}
      localJointOffset=<-5 0.2 -0.1>
     }]
    relativeSpaceBone {boneName=root}
   }
   n54 = Anim Dynamics - Bone: dyn_Cloth_volume_r_C02!(ComponentPose=@n53, Alpha=1, GravityOverride:Vector="0,0,-980", ExternalForce="0,0,0") -> (Pose) {
    angularDampingOverride=0.99
    angularSpringConstant=800
    bAngularSpring=true
    bOverrideAngularDamping=true
    bUseGravityOverride=true
    bUsePlanarLimit=false
    gravityScale=0.3
    simulationSpace=BoneRelative
    alphaBoolBlend {blendOption=Linear}
    boundBone {boneName=dyn_Cloth_volume_r_C02}
    componentAppliedLinearAccClamp=<2000 2000 1600>
    componentLinearAccScale=<0.5 0.5 0.4>
    componentLinearVelScale=<0.55 0.55 0.55>
    gravityOverride {z=-980}
    physicsBodyDefinitions [{
      collisionType=CustomSphere
      sphereCollisionRadius=2
      boundBone {boneName=dyn_Cloth_volume_r_C02}
      boxExtents=<33 33 33>
      constraintSetup {angularLimitsMax=<0 0 5>, angularLimitsMin=<0 0 0>, angularTarget=<0 0 0>}
      localJointOffset=<-5 0.2 -0.1>
     }]
    relativeSpaceBone {boneName=root}
   }
   n55 = "Constraint - dyn_Cloth_volume_bk (PP)"(ComponentPose=@n54, Alpha=1, ConstraintWeights_0=0.5, ConstraintWeights_1=0.5) {
    alphaBoolBlend {blendOption=Linear}
    boneToModify {boneName=dyn_Cloth_volume_bk}
    constraintSetup [{targetBone {boneName=dyn_Cloth_volume_l_C}}, {targetBone {boneName=dyn_Cloth_volume_r_C}}]
    constraintWeights [1,1]
   }
   n56 = "Constraint - dyn_Cloth_volume_bk01 (PP)"(ComponentPose=@n55, Alpha=1, ConstraintWeights_0=0.5, ConstraintWeights_1=0.5) {
    alphaBoolBlend {blendOption=Linear}
    boneToModify {boneName=dyn_Cloth_volume_bk01}
    constraintSetup [{targetBone {boneName=dyn_Cloth_volume_l_C01}}, {targetBone {boneName=dyn_Cloth_volume_r_C01}}]
    constraintWeights [1,1]
   }
   n57 = "Constraint - dyn_Cloth_volume_bk02 (PP)"(ComponentPose=@n56, Alpha=1, ConstraintWeights_0=0.5, ConstraintWeights_1=0.5) {
    alphaBoolBlend {blendOption=Linear}
    boneToModify {boneName=dyn_Cloth_volume_bk02}
    constraintSetup [{targetBone {boneName=dyn_Cloth_volume_l_C02}}, {targetBone {boneName=dyn_Cloth_volume_r_C02}}]
    constraintWeights [1,1]
   }
   n58 = RigidBody!(ComponentPose=@n57, Alpha=0.65, OverrideWorldGravity:Vector="0,0,-980", ExternalForce) -> (Pose) {
    bClampLinearTranslationLimitToRefPose=true
    bOverrideWorldGravity=true
    overridePhysicsAsset=&/M_MED_BananaPhilosopher_Sleek_Physics
    simulationSpace=BaseBoneSpace
    alphaBoolBlend {blendOption=Linear}
    baseBoneRef {boneName=gravity}
    componentAppliedLinearAccClamp=<1200 1200 1000>
    componentLinearAccScale=<0.01 0.15 0.15>
    componentLinearVelScale=<0.08 0.4 0.4>
    overrideWorldGravity {z=-980}
   }
   n59 = RigidBody(ComponentPose=@n58, OverrideWorldGravity="0,0,-980") {
    bClampLinearTranslationLimitToRefPose=true
    bOverrideWorldGravity=true
    overridePhysicsAsset=&/M_MED_BananaPhilosopher_Cloth_spine_Physics
    simulationSpace=BaseBoneSpace
    alphaBoolBlend {blendOption=Linear}
    baseBoneRef {boneName=pelvis}
    componentAppliedLinearAccClamp=<1200 1200 1000>
    componentLinearAccScale=<0.01 0.15 0.15>
    componentLinearVelScale=<0.08 0.4 0.4>
    overrideWorldGravity {z=-980}
   }
   n60 = "Constraint - dyn_brooch_base (PPPPPP)"!(
    ComponentPose=@n59
    Alpha=1
    ConstraintWeights_0=0.3114
    ConstraintWeights_1:float=0.2841
    ConstraintWeights_2:float=0.1284
    ConstraintWeights_3:float=0.108
    ConstraintWeights_4:float=0.0863
    ConstraintWeights_5:float=0.0817
   ) -> (Pose) {
    alphaBoolBlend {blendOption=Linear}
    boneToModify {boneName=dyn_brooch_base}
    constraintSetup [
     {targetBone {boneName=clavicle_l}}
     {targetBone {boneName=spine_04}}
     {targetBone {boneName=spine_03}}
     {targetBone {boneName=dyn_chest_fix_l}}
     {targetBone {boneName=spine_05}}
     {targetBone {boneName=upperarm_twist_01_l}}
    ]
    constraintWeights [1,1,1,1,1,1]
   }
   Save cached pose 'SkirtDynPose'(Pose=Blend Poses by bool(BlendPose_0=Use cached pose 'SkirtCachePose'(), BlendPose_1=Component To Local(ComponentPose=@n60), bActiveValue=Get bIsSurfaceSwimming()) {blendType=Linear}) {cacheName=SkirtDynPose}
  ]}
 EventGraph {nodes [
   Event BlueprintUpdateAnimation()
   > n1 = Sequence!(execute) -> (then_0, then_1, then_2:Exec, then_3:Exec)
   n2 = Get Body Mesh() -> (Body Mesh:Skeletal Mesh Component Object Reference)
   > Set Shoulder_rot_f_r(Shoulder_rot_f_r:double=GetAlphaFromReferenceTransformDelta(SkeletalMeshComponent=@n2, BoneName=upperarm_twist_01_r, BaseName=spine_05, InputRange=(X=20,Y=90), TwistAxis="0,0,1")) -> (Output_Get:double)
   > Set Shoulder_rot_f_l(Shoulder_rot_f_l:double=GetAlphaFromReferenceTransformDelta(SkeletalMeshComponent=@n2, BoneName=upperarm_twist_01_l, BaseName=spine_05, InputRange=(X=30,Y=70), TwistAxis="0,0,0")) -> (Output_Get:double)
   > Set Shoulder_rot_up_r(Shoulder_rot_up_r:double=GetAlphaFromReferenceTransformDelta(SkeletalMeshComponent=@n2, BoneName=clavicle_r, BaseName=spine_05, DeltaComponentType=EulerY, InputRange=(X=0,Y=110), TwistAxis="0,0,1")) -> (Output_Get:double)
   > Set Shoulder_rot_up_l(Shoulder_rot_up_l:double=GetAlphaFromReferenceTransformDelta(SkeletalMeshComponent=@n2, BoneName=clavicle_l, BaseName=spine_05, DeltaComponentType=EulerY, InputRange=(X=0,Y=110), TwistAxis="0,0,1")) -> (Output_Get:double)
   > Set Shoulder_rot_bk_l(Shoulder_rot_bk_l:double=GetAlphaFromReferenceTransformDelta(
     SkeletalMeshComponent=@n2
     BoneName=upperarm_twist_01_r
     BaseName=clavicle_r
     DeltaComponentType=EulerY
     InputRange=(X=-40,Y=110)
     TwistAxis="0,0,1"
    )) -> (Output_Get:double)
   > Set Shoulder_rot_bk_r(Shoulder_rot_bk_r:double=GetAlphaFromReferenceTransformDelta(SkeletalMeshComponent=@n2, BoneName=upperarm_l, BaseName=clavicle_l, DeltaComponentType=EulerY, InputRange=(X=-40,Y=110), TwistAxis="0,0,1")) -> (Output_Get:double)
   > n3 = Branch(Condition=Get bIsPlayingEmote())
   > Set thigh_shape_fn_l(thigh_shape_fn_l:double=0) -> (Output_Get:double)
   Set thigh_shape_fn_l(
    execute=@n3.else
    thigh_shape_fn_l=GetAlphaFromReferenceTransformDelta(SkeletalMeshComponent=@n2, BoneName=thigh_l, BaseName=pelvis, InputRange=(X=0,Y=-150), TwistAxis="0,0,1")
   )
   > Set thigh_shape_fn_r(thigh_shape_fn_r:double=GetAlphaFromReferenceTransformDelta(SkeletalMeshComponent=@n2, BoneName=thigh_r, BaseName=pelvis, InputRange=(X=0,Y=-110), TwistAxis="0,0,1")) -> (Output_Get:double)
   Set Is Dive And DBNO(
    execute=@n1.then_2
    Is Dive And DBNO:bool=OR Boolean!(A=Get bIsSkydiveDiveMode(), B=Get bIsSkydiveFloating(), C:bool=Get bIsDBNO()) -> (ReturnValue)
   ) -> (Output_Get:bool)
   n4 = Get bIsKneeSliding()
   > Set Is KneeAndBaseball Sliding(Is KneeAndBaseball Sliding:bool=@n4) -> (Output_Get:bool)
   Set DBNO(execute=@n1.then_3, DBNO:double=SelectFloat(A=0.5, B=0, bPickA=Get bIsDBNO())) -> (Output_Get:double)
   n5 = Gravity Override State Function DEPRECATED(
    execute=@n1.then_1
    CharacterPartAnimInstance=Self-Reference()
    GameTimeInSeconds=GetGameTimeInSeconds()
    GravityOverrideParams=Get Gravity Override Params()
   )
   > Set Gravity Override Cape(Gravity Override Cape:Vector=@n5) -> (Output_Get:Vector)
   n6 = Event OnCharacterCustomizationLoaded()
   > n7 = Sequence()
   Set Character Mesh(execute=@n7.then_1, Character Mesh:Skeletal Mesh Component Object Reference=@n6.CustomBaseMesh) -> (Output_Get:Skeletal Mesh Component Object Reference)
   Set Body Mesh(execute=@n7.then_2, Body Mesh:Skeletal Mesh Component Object Reference=Get Array Item with EFort Custom Part Type(Array=@n6.CustomMeshParts)) -> (Output_Get:Skeletal Mesh Component Object Reference)
   GetSkeletalMesh!(self:Custom Character Part Object Reference=Get Array Item with EFort Custom Part Type(Array:Array of Custom Character Part Object References=@n6.CustomParts) -> (Output:Custom Character Part Object Reference)) -> (ReturnValue:Skeletal Mesh Object Reference)
   OR Boolean!(A=Get bIsBaseballSliding(), B=@n4) -> (ReturnValue)
  ]}
]
variables [
 Body Mesh {type=SkeletalMeshComponent}
 Character Mesh {type=SkeletalMeshComponent}
 DBNO {type=number}
 Gravity Override Cape {type=Vector}
 Gravity Override Params {
  type=GravityOverrideParamsStruct
  default {
   dBNO {jointName=gravity}
   dBNO_Carried {jointName=gravity}
   onGround_Moving {jointName=gravity}
   riding_Moving_Targeting {gravityOverride=<0 -200 -800>}
   riding_Standing {gravityOverride=<0 -350 -400>}
   skydive_DivingDown {gravityOverride=<0 0 -50>, windAmplitude=<0 0 0>, windFrequency=<0 0 0>}
   skydive_FloatingDown {jointName=pelvis, gravityOverride=<0 0 -50>, windAmplitude=<0 0 0>, windFrequency=<0 0 0>, windOffset=<0 0 0>}
   skydive_GoingUp {gravityOverride=<0 -600 -150>}
   skydive_Parachuting {gravityOverride=<0 -500 -800>, windAmplitude=<0 0 0>, windFrequency=<0 0 0>}
   skydive_Surfing {gravityOverride=<0 -500 -980>, windAmplitude=<0 0 0>, windFrequency=<0 0 0>}
   swimming {jointName=gravity}
  }
 }
 Is Dive And DBNO {type=boolean}
 Is KneeAndBaseball Sliding {type=boolean}
 Shoulder_rot_bk_l {type=number}
 Shoulder_rot_bk_r {type=number}
 Shoulder_rot_f_l {type=number}
 Shoulder_rot_f_r {type=number}
 Shoulder_rot_up_l {type=number}
 Shoulder_rot_up_r {type=number}
 thigh_shape_fn_l {type=number}
 thigh_shape_fn_r {type=number}
]
